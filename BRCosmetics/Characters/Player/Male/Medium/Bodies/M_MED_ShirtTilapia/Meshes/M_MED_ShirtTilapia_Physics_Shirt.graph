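# Physics asset for the shirt on the medium male ShirtTilapia player body, rigging its jacket flaps for secondary motion.
# Capsules cover pelvis, lower spine and thighs, while tiny Mstr spheres anchor the flaps through free-rotation joints to pelvis or spine_01.
# Dyn capsules hang from these anchors with limited swing: single back and side segments, two-segment front chains with tighter limits.
# Side anchors are asymmetric: right joins spine_01, left joins pelvis.
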PhysicsAsset /BRCosmetics/Characters/Player/Male/Medium/Bodies/M_MED_ShirtTilapia/Meshes/M_MED_ShirtTilapia_Physics_Shirt
bodies {
 Dyn_JacketBack [{shapeType=Capsule, length=10.9515, radius=5.4419, shapeName=Dyn_JacketBack_capsule, center=<14.1013 0 -5.2743>, rotation=<rot 0 0 90>}]
 Dyn_JacketBack_L [{shapeType=Capsule, length=10.9515, radius=5.4419, center=<-14.1013 0 5.2743>, rotation=<rot 0 0 -90>}]
 Dyn_JacketBack_R [{shapeType=Capsule, length=10.9515, radius=5.4419, shapeName=Dyn_JacketBack_capsule, center=<14.1013 0 -5.2743>, rotation=<rot 0 0 90>}]
 Dyn_JacketFrontA_L [{shapeType=Capsule, length=9.122, radius=4.5443, center=<-8.1724 0 4.065>, rotation=<rot 0 0 -90>}]
 Dyn_JacketFrontA_R [{shapeType=Capsule, length=9.122, radius=4.5443, shapeName=Dyn_JacketBack_capsule, center=<8.1724 0 -4.065>, rotation=<rot 0 0 90>}]
 Dyn_JacketFrontB_L [{shapeType=Capsule, length=7.2356, radius=4.5443, center=<-6.0107 0 3.9509>, rotation=<rot 0 0 -90>}]
 Dyn_JacketFrontB_R [{shapeType=Capsule, length=7.2356, radius=4.5443, shapeName=Dyn_JacketBack_capsule, center=<6.0107 0 -3.9509>, rotation=<rot 0 0 90>}]
 Dyn_JacketSide_L [{shapeType=Capsule, length=10.9515, radius=5.4419, center=<-14.1013 0 5.2743>, rotation=<rot 0 0 -90>}]
 Dyn_JacketSide_R [{shapeType=Capsule, length=10.9515, radius=5.4419, shapeName=Dyn_JacketBack_capsule, center=<14.1013 0 -5.2743>, rotation=<rot 0 0 90>}]
 Mstr_JacketBack [{shapeType=Sphere, radius=0.505, shapeName=Mstr_JacketBack_sphere}]
 Mstr_JacketBack_L [{shapeType=Sphere, radius=0.505}]
 Mstr_JacketBack_R [{shapeType=Sphere, radius=0.505, shapeName=Mstr_JacketBack_sphere}]
 Mstr_JacketFront_L [{shapeType=Sphere, radius=0.505}]
 Mstr_JacketFront_R [{shapeType=Sphere, radius=0.505, shapeName=Mstr_JacketBack_sphere}]
 Mstr_JacketSide_L [{shapeType=Sphere, radius=0.505}]
 Mstr_JacketSide_R [{shapeType=Sphere, radius=0.505, shapeName=Mstr_JacketBack_sphere}]
 pelvis [{shapeType=Capsule, length=5.7215, radius=11.666, shapeName=pelvis_capsule, center=<0 -0.2509 0>, rotation=<rot 0 -90 0>}]
 spine_01 [{shapeType=Capsule, length=5.4709, radius=11.0757, shapeName=pelvis_capsule, center=<0 0.1411 0>, rotation=<rot 0 -90 0>}]
 spine_02 [{shapeType=Capsule, length=5.0209, radius=11.0162, shapeName=pelvis_capsule, center=<0 0.1411 0>, rotation=<rot 0 -90 0>}]
 thigh_l [{shapeType=Capsule, length=25.7753, radius=8.1503, shapeName=thigh_l_capsule, center=<-12.3947 1.5823 0.9087>, rotation=<rot -90 179.9997 0>}]
 thigh_r [{shapeType=Capsule, length=25.7753, radius=8.1503, shapeName=thigh_r_capsule, center=<12.3947 -1.5823 -0.9087>, rotation=<rot -90 -0.0003 0>}]
}
constraints [
 {
  bone1Name=spine_01
  bone2Name=pelvis
  swing1LimitDegrees=45
  swing1Motion=Free
  swing2LimitDegrees=45
  swing2Motion=Free
  twistLimitDegrees=45
  twistMotion=Free
 }
 {
  bone1Name=spine_02
  bone2Name=spine_01
  swing1LimitDegrees=45
  swing1Motion=Free
  swing2LimitDegrees=45
  swing2Motion=Free
  twistLimitDegrees=45
  twistMotion=Free
 }
 {
  bone1Name=thigh_r
  bone2Name=pelvis
  swing1LimitDegrees=45
  swing1Motion=Free
  swing2LimitDegrees=45
  swing2Motion=Free
  twistLimitDegrees=45
  twistMotion=Free
 }
 {
  bone1Name=thigh_l
  bone2Name=pelvis
  swing1LimitDegrees=45
  swing1Motion=Free
  swing2LimitDegrees=45
  swing2Motion=Free
  twistLimitDegrees=45
  twistMotion=Free
 }
 {
  bone1Name=Mstr_JacketBack
  bone2Name=spine_01
  swing1LimitDegrees=45
  swing1Motion=Free
  swing2LimitDegrees=45
  swing2Motion=Free
  twistLimitDegrees=45
  twistMotion=Free
 }
 {
  bone1Name=Dyn_JacketBack
  bone2Name=Mstr_JacketBack
  swing1LimitDegrees=45
  swing1Motion=Locked
  swing2LimitDegrees=25
  twistLimitDegrees=45
  twistMotion=Locked
 }
 {
  bone1Name=Mstr_JacketBack_R
  bone2Name=pelvis
  swing1LimitDegrees=45
  swing1Motion=Free
  swing2LimitDegrees=45
  swing2Motion=Free
  twistLimitDegrees=45
  twistMotion=Free
 }
 {
  bone1Name=Dyn_JacketBack_R
  bone2Name=Mstr_JacketBack_R
  swing1LimitDegrees=10
  swing2LimitDegrees=25
  twistLimitDegrees=45
  twistMotion=Locked
 }
 {
  bone1Name=Mstr_JacketBack_L
  bone2Name=pelvis
  swing1LimitDegrees=45
  swing1Motion=Free
  swing2LimitDegrees=45
  swing2Motion=Free
  twistLimitDegrees=45
  twistMotion=Free
 }
 {
  bone1Name=Dyn_JacketBack_L
  bone2Name=Mstr_JacketBack_L
  swing1LimitDegrees=10
  swing2LimitDegrees=25
  twistLimitDegrees=45
  twistMotion=Locked
 }
 {
  bone1Name=Mstr_JacketSide_R
  bone2Name=spine_01
  swing1LimitDegrees=45
  swing1Motion=Free
  swing2LimitDegrees=45
  swing2Motion=Free
  twistLimitDegrees=45
  twistMotion=Free
 }
 {
  bone1Name=Dyn_JacketSide_R
  bone2Name=Mstr_JacketSide_R
  swing1LimitDegrees=45
  swing1Motion=Locked
  swing2LimitDegrees=45
  twistLimitDegrees=45
  twistMotion=Locked
 }
 {
  bone1Name=Mstr_JacketFront_R
  bone2Name=pelvis
  swing1LimitDegrees=45
  swing1Motion=Free
  swing2LimitDegrees=45
  swing2Motion=Free
  twistLimitDegrees=45
  twistMotion=Free
 }
 {bone1Name=Dyn_JacketFrontA_R, bone2Name=Mstr_JacketFront_R, swing1LimitDegrees=5, swing2LimitDegrees=15, twistLimitDegrees=0}
 {bone1Name=Dyn_JacketFrontB_R, bone2Name=Dyn_JacketFrontA_R, swing1LimitDegrees=5, swing2LimitDegrees=15, twistLimitDegrees=0}
 {
  bone1Name=Dyn_JacketSide_L
  bone2Name=Mstr_JacketSide_L
  swing1LimitDegrees=45
  swing1Motion=Locked
  swing2LimitDegrees=45
  twistLimitDegrees=45
  twistMotion=Locked
 }
 {
  bone1Name=Mstr_JacketSide_L
  bone2Name=pelvis
  swing1LimitDegrees=45
  swing1Motion=Free
  swing2LimitDegrees=45
  swing2Motion=Free
  twistLimitDegrees=45
  twistMotion=Free
 }
 {bone1Name=Dyn_JacketFrontB_L, bone2Name=Dyn_JacketFrontA_L, swing1LimitDegrees=5, swing2LimitDegrees=15, twistLimitDegrees=0}
 {bone1Name=Dyn_JacketFrontA_L, bone2Name=Mstr_JacketFront_L, swing1LimitDegrees=5, swing2LimitDegrees=15, twistLimitDegrees=0}
 {
  bone1Name=Mstr_JacketFront_L
  bone2Name=pelvis
  swing1LimitDegrees=45
  swing1Motion=Free
  swing2LimitDegrees=45
  swing2Motion=Free
  twistLimitDegrees=45
  twistMotion=Free
 }
]
