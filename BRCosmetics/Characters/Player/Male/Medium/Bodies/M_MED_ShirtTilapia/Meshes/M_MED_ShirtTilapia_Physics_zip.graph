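# Physics asset for the zipper on the male medium ShirtTilapia body cosmetic, covering only the zip bones and nearby body parts.
# A small box on Zip_dyn is constrained to a sphere on Zip_Base with angular limits, wide in swing and narrow in twist.
# Capsules on pelvis and right thigh give body shapes near the zip.
# Zip_Base and thigh_r both link to pelvis with free angular motion.
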PhysicsAsset /BRCosmetics/Characters/Player/Male/Medium/Bodies/M_MED_ShirtTilapia/Meshes/M_MED_ShirtTilapia_Physics_zip
bodies {
 Zip_Base [{shapeType=Sphere, radius=0.505, shapeName=Zip_Base_sphere, center=<-0.1307 0.1248 0.1339>}]
 Zip_dyn [{
   shapeType=Box
   extentX=1.01
   extentY=1.01
   extentZ=2.2088
   shapeName=Zip_dyn_box
   center=<1.5379 -0.0558 0.009>
   rotation=<rot 0 -91.1566 -89.6363>
  }]
 pelvis [{shapeType=Capsule, length=5.7215, radius=11.666, shapeName=pelvis_capsule, center=<0 -0.2509 0>, rotation=<rot 0 -90 0>}]
 thigh_r [{shapeType=Capsule, length=25.7753, radius=8.1503, shapeName=thigh_r_capsule, center=<12.3947 -1.5823 -0.9087>, rotation=<rot -90 -0.0003 0>}]
}
constraints [
 {
  bone1Name=thigh_r
  bone2Name=pelvis
  swing1LimitDegrees=45
  swing1Motion=Free
  swing2LimitDegrees=45
  swing2Motion=Free
  twistLimitDegrees=45
  twistMotion=Free
 }
 {
  bone1Name=Zip_Base
  bone2Name=pelvis
  swing1LimitDegrees=45
  swing1Motion=Free
  swing2LimitDegrees=45
  swing2Motion=Free
  twistLimitDegrees=45
  twistMotion=Free
 }
 {bone1Name=Zip_dyn, bone2Name=Zip_Base, swing1LimitDegrees=88.2, swing2LimitDegrees=66.6, twistLimitDegrees=5}
]
