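# Animation Blueprint for the male medium ShirtTilapia body cosmetic, adding secondary motion to the shirt and its zipper.
# It copies the attached parent's pose, runs a Control Rig, then simulates shirt RigidBody physics, constrains the zip base to the jacket front and simulates the zip.
# Each update fades shirt physics out during crouching or emotes, feeds speed to the rig, and eases in upward force while airborne without parachute, stronger when skydiving.
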AnimBlueprint /BRCosmetics/Characters/Player/Male/Medium/Bodies/M_MED_ShirtTilapia/Meshes/M_MED_ShirtTilapia_AnimBP
target_skeleton=&/M_MED_ShirtTilapia_Skeleton
graphs [
 AnimGraph {nodes [
   n1 = Control Rig!(
    Source=Copy Pose From Mesh() {bUseAttachedParent=true}
    Alpha
    crouch:bool=Get bIsCrouching()
    Velocity:double=Get Velocity Pawn() -> (Velocity Pawn:double)
    RB alpha:double=Get RB Alpha() -> (RB Alpha:double)
   ) -> (Pose) {
    alphaBoolBlend {blendOption=Linear}
    controlRigAssetReference {blueprintRigClass=&/M_MED_ShirtTilapia_CtrlRig.M_MED_ShirtTilapia_CtrlRig_C}
   }
   n2 = RigidBody(ComponentPose=Local To Component(LocalPose=@n1), Alpha=Get RB Alpha(), ExternalForce=Get ExternalForce() -> (ExternalForce:Vector)) {
    overridePhysicsAsset=&/M_MED_ShirtTilapia_Physics_Shirt
    simulationSpace=BaseBoneSpace
    alphaBoolBlend {blendOption=Linear}
    baseBoneRef {boneName=pelvis}
    componentAppliedLinearAccClamp=<1000 1000 1000>
    componentLinearAccScale=<0.2 0.2 0.2>
    componentLinearVelScale=<0.2 0.2 0.2>
    simSpaceSettings {maxAngularAcceleration=150, maxAngularVelocity=10, maxLinearAcceleration=450, maxLinearVelocity=40, worldAlpha=1}
   }
   n3 = "Constraint - Zip_Base (P)"(ComponentPose=@n2, Alpha=1, ConstraintWeights_0=1) {
    alphaBoolBlend {blendOption=Linear}
    boneToModify {boneName=Zip_Base}
    constraintSetup [{targetBone {boneName=Dyn_JacketFrontB_R}}]
    constraintWeights [1]
   }
   n4 = RigidBody(ComponentPose=@n3) {
    overridePhysicsAsset=&/M_MED_ShirtTilapia_Physics_zip
    simulationSpace=BaseBoneSpace
    alphaBoolBlend {blendOption=Linear}
    baseBoneRef {boneName=Dyn_JacketFrontB_R}
    overrideWorldGravity {z=-980}
    simSpaceSettings {maxAngularAcceleration=150, maxAngularVelocity=10, maxLinearAcceleration=150, maxLinearVelocity=40, worldAlpha=1}
   }
   Output Pose(Result=Component To Local(ComponentPose=@n4)) {layerGroup=None}
  ]}
 EventGraph {nodes [
   n1 = Event BlueprintUpdateAnimation()
   > Set Velocity Pawn(Velocity Pawn:double=Vector Length(A=Get PawnVelocity())) -> (Output_Get:double)
   n2 = NOT Boolean(A=OR Boolean(A=Get bIsCrouching(), B=Get bIsPlayingEmote()))
   > Set RB Alpha(RB Alpha:double=FInterpTo(Current=Get RB Alpha(), Target="To Float (Boolean)"(InBool=@n2), DeltaTime=@n1.DeltaTimeX, InterpSpeed=12)) -> (Output_Get:double)
   n3 = SelectFloat(A=1200, B=900, bPickA=Get bIsSkyDiving())
   n4 = AND Boolean(A=OR Boolean(A=Get bIsFalling(), B=Get bIsJumping()), B=NOT Boolean(A=Get bIsParachuteOpened()))
   > Set ExternalForce(ExternalForce:Vector=VInterpTo!(
     Current:Vector=Get ExternalForce()
     Target_X:double=0
     Target_Y:double=0
     Target_Z:double=SelectFloat(A=@n3, B=0, bPickA=@n4)
     DeltaTime:float=@n1.DeltaTimeX
     InterpSpeed:float=5
    ) -> (ReturnValue:Vector)) -> (Output_Get:Vector)
  ]}
]
variables [
 ExternalForce {type=Vector}
 RB Alpha {type=number}
 Velocity Pawn {type=number}
]
